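# Sound cue for equipping the Moonflax Combat AR weapon.
# A Random node picks between two equip waves at equal weight and feeds a root Modulator left at default settings.
# The cue plays in the weapon foley sound class at reduced volume, using the shared weapon reload attenuation.
# It overrides concurrency and has its debug flag switched on.
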SoundCue /WeaponsUpdated/Gameplay/CombatAR/Sounds/Moonflax_CombatAR_Equip_Cue
root=Modulator_0
nodes [
 {id=Modulator_0, children [Random_0]}
 {id=Random_0, children [WavePlayer_2,WavePlayer_3], properties {weights [1,1]}}
 {id=WavePlayer_2, properties {soundWaveAssetPtr=&/Moonflax_CombatAR_Equip_01}}
 {id=WavePlayer_3, properties {soundWaveAssetPtr=&/Moonflax_CombatAR_Equip_02}}
]
properties {
 attenuationSettings=&Fort_Weapon_Reload_Att
 bDebug=true
 bOverrideConcurrency=true
 maxDistance=2250
 soundClassObject=&Fort_WeaponFoley
 volumeMultiplier=0.652
}
