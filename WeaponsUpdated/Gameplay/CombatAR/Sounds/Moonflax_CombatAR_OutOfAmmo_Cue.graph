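# Sound cue for the dry-fire sound of the Moonflax Combat AR when it is out of ammo.
# A single wave player feeds the dry-fire sample through a Modulator left at its default settings.
# The cue plays quietly in the weapon foley sound class, borrows the weapon reload attenuation and overrides concurrency.
# Its debug flag is switched on.
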SoundCue /WeaponsUpdated/Gameplay/CombatAR/Sounds/Moonflax_CombatAR_OutOfAmmo_Cue
root=Modulator_0
nodes [
 {id=Modulator_0, children [WavePlayer_0]}
 {id=WavePlayer_0, properties {soundWaveAssetPtr=&/Moonflax_Combat_AR_Dryfire_01}}
]
properties {
 attenuationSettings=&Fort_Weapon_Reload_Att
 bDebug=true
 bOverrideConcurrency=true
 maxDistance=2250
 soundClassObject=&Fort_WeaponFoley
 volumeMultiplier=0.1
}
